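# Physics asset for the GrimHound player head cosmetic, named for the frontend idle, that sets up physics for its two dangling bands.
# Narrow capsules form a six-bone left chain and a four-bone right chain from the mane base, linked by tight swing and twist limits, looser at the left tip.
# Unconstrained capsules cover mane_5, upper spine, neck, shoulders and upper arms, plus an oversized, offset head capsule.
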PhysicsAsset /BRCosmetics/Characters/Player/Male/Medium/Heads/M_MED_GrimHound_Head/Meshes/M_MED_GrimHound_Head_Frontend_Idle_Physics
bodies {
 clavicle_l [{shapeType=Capsule, length=2, radius=14, shapeName=clavicle_r_capsule, center=<9 0 -2.1539>, rotation=<rot 80 -0.0011 -15.0011>}]
 clavicle_r [{shapeType=Capsule, length=2, radius=14, shapeName=clavicle_r_capsule, center=<-9 0 2.1539>, rotation=<rot 80 -0.001 -15.0011>}]
 dyn_L_band_1 [{shapeType=Capsule, length=7, radius=1, shapeName=dyn_L_band_1_capsule, center=<6 0 0>, rotation=<rot 90 0 0>}]
 dyn_L_band_2 [{shapeType=Capsule, length=7, radius=1, shapeName=dyn_L_band_2_capsule, center=<6 0 0>, rotation=<rot 90 0 0>}]
 dyn_L_band_3 [{shapeType=Capsule, length=7, radius=0.8, shapeName=dyn_L_band_3_capsule, center=<6 0 0>, rotation=<rot 90 0 0>}]
 dyn_L_band_4 [{shapeType=Capsule, length=7, radius=0.8, shapeName=dyn_L_band_4_capsule, center=<6 0 0>, rotation=<rot 90 0 0>}]
 dyn_L_band_5 [{shapeType=Capsule, length=7, radius=0.8, shapeName=dyn_L_band_5_capsule, center=<6 0 0>, rotation=<rot 90 0 0>}]
 dyn_L_band_6 [{shapeType=Capsule, length=13, radius=0.8, shapeName=dyn_L_band_6_capsule, center=<7 0 0>, rotation=<rot 90 0 0>}]
 dyn_R_band_1 [{shapeType=Capsule, length=7, radius=1, shapeName=dyn_R_band_1_capsule, center=<6 0 0>, rotation=<rot 90 0 0>}]
 dyn_R_band_2 [{shapeType=Capsule, length=7, radius=1, shapeName=dyn_R_band_2_capsule, center=<6 0 0>, rotation=<rot 90 0 0>}]
 dyn_R_band_3 [{shapeType=Capsule, length=7, radius=0.8, shapeName=dyn_R_band_3_capsule, center=<6 0 0>, rotation=<rot 90 0 0>}]
 dyn_R_band_4 [{shapeType=Capsule, length=8, radius=0.8, shapeName=dyn_R_band_4_capsule, center=<6 0 0>, rotation=<rot 90 0 0>}]
 dyn_mane_5 [{shapeType=Capsule, length=5.7335, radius=4, shapeName=dyn_mane_5_capsule, center=<1.5886 1.3292 0.1207>, rotation=<rot 85 -45.0001 0>}]
 dyn_mane_base [{shapeType=Capsule, length=1, radius=1, shapeName=dyn_mane_base_capsule}]
 head [{shapeType=Capsule, length=60, radius=60, shapeName=head_capsule, center=<-5.7195 52.2053 0>}]
 neck_01 [{shapeType=Capsule, length=5, radius=14, shapeName=neck_01_capsule, center=<0 -0.3892 0>, rotation=<rot 90 0 0>}]
 spine_04 [{shapeType=Capsule, length=5, radius=16, shapeName=spine_04_capsule}]
 spine_05 [{shapeType=Capsule, length=5, radius=19, shapeName=spine_05_capsule, center=<2.4799 1.4305 0>}]
 upperarm_l [{shapeType=Capsule, length=10, radius=9, shapeName=upperarm_l_capsule, center=<7 0 0>, rotation=<rot 90 0 0>}]
 upperarm_r [{shapeType=Capsule, length=10, radius=9, shapeName=upperarm_l_capsule, center=<-7 0 0>, rotation=<rot 90 0 0>}]
}
constraints [
 {bone1Name=dyn_L_band_1, bone2Name=dyn_mane_base, swing1LimitDegrees=10, swing2LimitDegrees=10, twistLimitDegrees=10}
 {bone1Name=dyn_L_band_2, bone2Name=dyn_L_band_1, swing1LimitDegrees=10, swing2LimitDegrees=10, twistLimitDegrees=10}
 {bone1Name=dyn_L_band_3, bone2Name=dyn_L_band_2, swing1LimitDegrees=10, swing2LimitDegrees=10, twistLimitDegrees=10}
 {bone1Name=dyn_L_band_4, bone2Name=dyn_L_band_3, swing1LimitDegrees=10, swing2LimitDegrees=10, twistLimitDegrees=10}
 {bone1Name=dyn_L_band_5, bone2Name=dyn_L_band_4, swing1LimitDegrees=10, swing2LimitDegrees=10, twistLimitDegrees=10}
 {bone1Name=dyn_L_band_6, bone2Name=dyn_L_band_5, swing1LimitDegrees=20, swing2LimitDegrees=20, twistLimitDegrees=20}
 {bone1Name=dyn_R_band_1, bone2Name=dyn_mane_base, swing1LimitDegrees=10, swing2LimitDegrees=10, twistLimitDegrees=10}
 {bone1Name=dyn_R_band_2, bone2Name=dyn_R_band_1, swing1LimitDegrees=10, swing2LimitDegrees=10, twistLimitDegrees=10}
 {bone1Name=dyn_R_band_3, bone2Name=dyn_R_band_2, swing1LimitDegrees=10, swing2LimitDegrees=10, twistLimitDegrees=10}
 {bone1Name=dyn_R_band_4, bone2Name=dyn_R_band_3, swing1LimitDegrees=10, swing2LimitDegrees=10, twistLimitDegrees=10}
]
