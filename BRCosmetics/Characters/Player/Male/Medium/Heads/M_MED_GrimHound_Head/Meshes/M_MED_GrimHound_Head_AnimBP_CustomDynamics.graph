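# Post-process style Anim Blueprint for the GrimHound male head cosmetic, adding secondary motion to ears, mane, tufts and bands.
# Layers body-mesh hood and clavicles, then applies Control Rig, weighted tuft constraints, band offsets and AnimDynamics into a cached pose.
# That pose feeds chained head and backbling RigidBody sims, or a separate idle sim in the front end.
# EventGraph caches body mesh, skydive gravity, variant, backbling and low-setting state.
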AnimBlueprint /BRCosmetics/Characters/Player/Male/Medium/Heads/M_MED_GrimHound_Head/Meshes/M_MED_GrimHound_Head_AnimBP_CustomDynamics
target_skeleton=&/M_MED_GrimHound_Base_Head_Export_Skeleton
graphs [
 AnimGraph {nodes [
   n1 = RigidBody!(ComponentPose=Local To Component(LocalPose=Use cached pose 'preRBAN'()), Alpha, OverrideWorldGravity:Vector="0,0,-980", ExternalForce) -> (Pose) {
    bClampLinearTranslationLimitToRefPose=true
    bOverrideWorldGravity=true
    overridePhysicsAsset=&/M_MED_GrimHound_Head_Physics
    simulationSpace=BaseBoneSpace
    alphaBoolBlend {blendOption=Linear}
    alphaScaleBiasClamp {bias=1, scale=-1}
    baseBoneRef {boneName=spine_04}
    overrideWorldGravity {z=-980}
    simSpaceSettings {
     dampingAlpha=0.05
     maxAngularAcceleration=50
     maxAngularVelocity=7
     maxLinearAcceleration=4000
     maxLinearVelocity=800
     velocityScaleZ=0.45
     worldAlpha=0.75
     externalAngularVelocity=<1 1 1>
     externalLinearDragV=<1 1 1>
     externalLinearVelocity=<1 1 1>
    }
   }
   n2 = RigidBody(ComponentPose=@n1, OverrideWorldGravity="0,0,-980") {
    bClampLinearTranslationLimitToRefPose=true
    bOverrideWorldGravity=true
    overridePhysicsAsset=&/M_MED_GrimHound_Head_Backbling_Physics
    simulationSpace=BaseBoneSpace
    alphaBoolBlend {blendOption=Linear}
    baseBoneRef {boneName=neck_02}
    overrideWorldGravity {z=-980}
    simSpaceSettings {
     dampingAlpha=0.05
     maxAngularAcceleration=50
     maxAngularVelocity=7
     maxLinearAcceleration=4000
     maxLinearVelocity=800
     velocityScaleZ=0
     worldAlpha=0.75
     externalAngularVelocity=<1 1 1>
     externalLinearDragV=<1 1 1>
     externalLinearVelocity=<1 1 1>
    }
   }
   n3 = RigidBody(ComponentPose=Local To Component(LocalPose=Use cached pose 'preRBAN'()), OverrideWorldGravity="0,0,-980") {
    bClampLinearTranslationLimitToRefPose=true
    bOverrideWorldGravity=true
    overridePhysicsAsset=&/M_MED_GrimHound_Head_Frontend_Idle_Physics
    simulationSpace=BaseBoneSpace
    alphaBoolBlend {blendOption=Linear}
    baseBoneRef {boneName=spine_04}
    overrideWorldGravity {z=-980}
    simSpaceSettings {
     dampingAlpha=0.05
     maxAngularAcceleration=50
     maxAngularVelocity=7
     maxLinearAcceleration=4000
     maxLinearVelocity=800
     velocityScaleZ=0.45
     worldAlpha=0.75
     externalAngularVelocity=<1 1 1>
     externalLinearDragV=<1 1 1>
     externalLinearVelocity=<1 1 1>
    }
   }
   Output Pose(Result=Blend Poses by bool(BlendPose_0=Component To Local(ComponentPose=@n3), BlendPose_1=Component To Local(ComponentPose=@n2), bActiveValue=Get bIsInFrontEnd()) {blendType=Linear})
   n4 = Layered blend per bone(BasePose=Input Pose(), BlendPoses_0=Copy Pose From Mesh(SourceMeshComponent=Get Body_Mesh()) {bUseAttachedParent=true}) {layerSetup [{branchFilters [{boneName=dyn_hood}, {boneName=clavicle_l}, {boneName=clavicle_r}]}]}
   n5 = Control Rig(Source=@n4) {
    alphaBoolBlend {blendOption=Linear}
    controlRigAssetReference {blueprintRigClass=&/M_MED_GrimHound_Head_CtrlRig.M_MED_GrimHound_Head_CtrlRig_C}
   }
   n6 = "Constraint - dyn_tuftPlaceholder_1 (PPPPPP)"!(
    ComponentPose=Local To Component(LocalPose=@n5)
    Alpha=1
    ConstraintWeights_0=0.0171
    ConstraintWeights_1:float=0.425
    ConstraintWeights_2:float=0.3819
    ConstraintWeights_3:float=0.0367
    ConstraintWeights_4:float=0.1362
    ConstraintWeights_5:float=0.0031
   ) -> (Pose) {
    boneToModify {boneName=dyn_tuftPlaceholder_1}
    constraintSetup [
     {targetBone {boneName=dyn_mane_4}}
     {targetBone {boneName=dyn_mane_base}}
     {targetBone {boneName=head}}
     {targetBone {boneName=neck_01}}
     {targetBone {boneName=neck_02}}
     {targetBone {boneName=dyn_front_mane_base}}
    ]
    constraintWeights [1,1,1,1,1,1]
   }
   n7 = "Constraint - dyn_tuftPlaceholder_2 (PPPPPPP)"!(
    ComponentPose=@n6
    Alpha=1
    ConstraintWeights_0=0.0674
    ConstraintWeights_1:float=0.0207
    ConstraintWeights_2:float=0.0519
    ConstraintWeights_3:float=0.7688
    ConstraintWeights_4:float=0.0586
    ConstraintWeights_5:float=0.0061
    ConstraintWeights_6:float=0.0265
   ) -> (Pose) {
    boneToModify {boneName=dyn_tuftPlaceholder_2}
    constraintSetup [
     {targetBone {boneName=dyn_mane_5}}
     {targetBone {boneName=spine_05}}
     {targetBone {boneName=dyn_mane_4}}
     {targetBone {boneName=dyn_mane_base}}
     {targetBone {boneName=head}}
     {targetBone {boneName=neck_01}}
     {targetBone {boneName=neck_02}}
    ]
    constraintWeights [1,1,1,1,1,1,1]
   }
   n8 = "Constraint - dyn_tuftPlaceholder_3 (PPPPP)"!(
    ComponentPose=@n7
    Alpha=1
    ConstraintWeights_0=0.0272
    ConstraintWeights_1:float=0.4902
    ConstraintWeights_2:float=0.3147
    ConstraintWeights_3:float=0.0352
    ConstraintWeights_4:float=0.1326
   ) -> (Pose) {
    boneToModify {boneName=dyn_tuftPlaceholder_3}
    constraintSetup [
     {targetBone {boneName=dyn_mane_4}}
     {targetBone {boneName=dyn_mane_base}}
     {targetBone {boneName=head}}
     {targetBone {boneName=neck_01}}
     {targetBone {boneName=neck_02}}
    ]
    constraintWeights [1,1,1,1,1]
   }
   n9 = "Constraint - dyn_tuftPlaceholder_4 (PPPPPP)"!(
    ComponentPose=@n8
    Alpha=1
    ConstraintWeights_0=0.103
    ConstraintWeights_1:float=0.0187
    ConstraintWeights_2:float=0.1509
    ConstraintWeights_3:float=0.6957
    ConstraintWeights_4:float=0.0217
    ConstraintWeights_5:float=0.0101
   ) -> (Pose) {
    boneToModify {boneName=dyn_tuftPlaceholder_4}
    constraintSetup [
     {targetBone {boneName=dyn_mane_5}}
     {targetBone {boneName=spine_05}}
     {targetBone {boneName=dyn_mane_4}}
     {targetBone {boneName=dyn_mane_base}}
     {targetBone {boneName=head}}
     {targetBone {boneName=neck_02}}
    ]
    constraintWeights [1,1,1,1,1,1]
   }
   n10 = "Transform (Modify) Bone - Bone: dyn_L_band_1"(ComponentPose=@n9, Alpha=1, Translation="0,0,0", Rotation="-15,0,0", Scale="1,1,1") {
    rotationMode=BMM_Additive
    rotationSpace=BCS_BoneSpace
    alphaBoolBlend {blendOption=Linear}
    boneToModify {boneName=dyn_L_band_1}
    rotation {pitch=-15}
   }
   n11 = "Transform (Modify) Bone - Bone: dyn_R_band_1"(ComponentPose=@n10, Alpha=1, Translation="0,0,0", Rotation="-15,0,0", Scale="1,1,1") {
    rotationMode=BMM_Additive
    rotationSpace=BCS_BoneSpace
    alphaBoolBlend {blendOption=Linear}
    boneToModify {boneName=dyn_R_band_1}
    rotation {pitch=-15}
   }
   n12 = "Transform (Modify) Bone - Bone: dyn_L_band_3"(ComponentPose=@n11, Alpha=1, Translation="0,0,0", Rotation="-12,0,0", Scale="1,1,1") {
    rotationMode=BMM_Additive
    rotationSpace=BCS_BoneSpace
    alphaBoolBlend {blendOption=Linear}
    boneToModify {boneName=dyn_L_band_3}
    rotation {pitch=-12}
   }
   n13 = "Transform (Modify) Bone - Bone: dyn_L_band_5"(ComponentPose=@n12, Alpha=1, Translation="0,0,0", Rotation="-20,0,0", Scale="1,1,1") {
    rotationMode=BMM_Additive
    rotationSpace=BCS_BoneSpace
    alphaBoolBlend {blendOption=Linear}
    boneToModify {boneName=dyn_L_band_5}
    rotation {pitch=-20}
   }
   n14 = "Constraint - dyn_L_band_2 (P)"(ComponentPose=@n13, Alpha=1, ConstraintWeights_0=1) {
    alphaBoolBlend {blendOption=Linear}
    alphaScaleBiasClamp {scale=0.4}
    boneToModify {boneName=dyn_L_band_2}
    constraintSetup [{targetBone {boneName=spine_05}}]
    constraintWeights [1]
   }
   n15 = "Constraint - dyn_R_band_2 (P)"(ComponentPose=@n14, Alpha=1, ConstraintWeights_0=1) {
    alphaBoolBlend {blendOption=Linear}
    alphaScaleBiasClamp {scale=0.4}
    boneToModify {boneName=dyn_R_band_2}
    constraintSetup [{targetBone {boneName=spine_05}}]
    constraintWeights [1]
   }
   n16 = Anim Dynamics - Bone: dyn_L_ear(ComponentPose=@n15, Alpha=1, ExternalForce="0,0,0") {
    angularDampingOverride=1
    angularSpringConstant=9000
    bAngularSpring=true
    bOverrideAngularDamping=true
    bUseGravityOverride=true
    gravityScale=0.35
    simulationSpace=RootRelative
    alphaBoolBlend {blendOption=Linear}
    boundBone {boneName=dyn_L_ear}
    componentAppliedLinearAccClamp=<5000 5000 5000>
    componentLinearAccScale=<1 1 1>
    componentLinearVelScale=<1 1 1>
    gravityOverride {z=-980}
    physicsBodyDefinitions [{
      boundBone {boneName=dyn_L_ear}
      boxExtents=<30 30 30>
      constraintSetup {angularLimitsMax=<15 0 15>, angularLimitsMin=<-15 -15 -5>, angularTarget=<1 0 0>}
      localJointOffset=<4 0 0>
     }]
   }
   n17 = Anim Dynamics - Bone: dyn_R_ear(ComponentPose=@n16, Alpha=1, ExternalForce="0,0,0") {
    angularDampingOverride=1
    angularSpringConstant=9000
    bAngularSpring=true
    bOverrideAngularDamping=true
    bUseGravityOverride=true
    gravityScale=0.35
    simulationSpace=RootRelative
    alphaBoolBlend {blendOption=Linear}
    boundBone {boneName=dyn_R_ear}
    componentAppliedLinearAccClamp=<5000 5000 5000>
    componentLinearAccScale=<1 1 1>
    componentLinearVelScale=<1 1 1>
    gravityOverride {z=-980}
    physicsBodyDefinitions [{
      boundBone {boneName=dyn_R_ear}
      boxExtents=<30 30 30>
      constraintSetup {angularLimitsMax=<15 0 15>, angularLimitsMin=<-15 -15 -5>, angularTarget=<1 0 0>}
      localJointOffset=<-4 0 0>
     }]
   }
   n18 = Anim Dynamics - Bone: dyn_mane_1(ComponentPose=@n17, Alpha=1, ExternalForce="0,0,0") {
    angularDampingOverride=1
    angularSpringConstant=9000
    bAngularSpring=true
    bOverrideAngularDamping=true
    bUseGravityOverride=true
    gravityScale=0.35
    simulationSpace=RootRelative
    alphaBoolBlend {blendOption=Linear}
    boundBone {boneName=dyn_mane_1}
    componentAppliedLinearAccClamp=<5000 5000 5000>
    componentLinearAccScale=<1 1 1>
    componentLinearVelScale=<1 1 1>
    gravityOverride {z=-980}
    physicsBodyDefinitions [{
      boundBone {boneName=dyn_mane_1}
      boxExtents=<20 20 20>
      constraintSetup {angularLimitsMax=<15 15 20>, angularLimitsMin=<-15 -15 -20>, angularTarget=<1 0 0>}
      localJointOffset=<4 0 0>
     }]
    simSpaceSettings {maxAngularAcceleration=35, maxAngularVelocity=7, simSpaceAngularAlpha=1}
   }
   n19 = Anim Dynamics - Bone: dyn_mane_2(ComponentPose=@n18, Alpha=1, ExternalForce="0,0,0") {
    angularDampingOverride=1
    angularSpringConstant=9000
    bAngularSpring=true
    bOverrideAngularDamping=true
    bUseGravityOverride=true
    gravityScale=0.35
    simulationSpace=RootRelative
    alphaBoolBlend {blendOption=Linear}
    boundBone {boneName=dyn_mane_2}
    componentAppliedLinearAccClamp=<5000 5000 5000>
    componentLinearAccScale=<1 1 1>
    componentLinearVelScale=<1 1 1>
    gravityOverride {z=-980}
    physicsBodyDefinitions [{
      boundBone {boneName=dyn_mane_2}
      boxExtents=<20 20 20>
      constraintSetup {angularLimitsMax=<15 15 20>, angularLimitsMin=<-15 -15 -20>, angularTarget=<1 0 0>}
      localJointOffset=<4 0 0>
     }]
    simSpaceSettings {maxAngularAcceleration=35, maxAngularVelocity=7, simSpaceAngularAlpha=1}
   }
   n20 = Anim Dynamics - Bone: dyn_mane_3(ComponentPose=@n19, Alpha=1, ExternalForce="0,0,0") {
    angularDampingOverride=1
    angularSpringConstant=9000
    bAngularSpring=true
    bOverrideAngularDamping=true
    bUseGravityOverride=true
    gravityScale=0.35
    simulationSpace=RootRelative
    alphaBoolBlend {blendOption=Linear}
    boundBone {boneName=dyn_mane_3}
    componentAppliedLinearAccClamp=<5000 5000 5000>
    componentLinearAccScale=<1 1 1>
    componentLinearVelScale=<1 1 1>
    gravityOverride {z=-980}
    physicsBodyDefinitions [{
      boundBone {boneName=dyn_mane_3}
      boxExtents=<20 20 20>
      constraintSetup {angularLimitsMax=<15 15 15>, angularLimitsMin=<-15 -15 -15>, angularTarget=<1 0 0>}
      localJointOffset=<4 0 0>
     }]
    simSpaceSettings {maxAngularAcceleration=35, maxAngularVelocity=7, simSpaceAngularAlpha=1}
   }
   n21 = Anim Dynamics - Bone: dyn_mane_4(ComponentPose=@n20, Alpha=1, ExternalForce="0,0,0") {
    angularDampingOverride=1
    angularSpringConstant=9000
    bAngularSpring=true
    bOverrideAngularDamping=true
    bUseGravityOverride=true
    gravityScale=0.35
    simulationSpace=RootRelative
    alphaBoolBlend {blendOption=Linear}
    boundBone {boneName=dyn_mane_4}
    componentAppliedLinearAccClamp=<5000 5000 5000>
    componentLinearAccScale=<1 1 1>
    componentLinearVelScale=<1 1 1>
    gravityOverride {z=-980}
    physicsBodyDefinitions [{
      boundBone {boneName=dyn_mane_4}
      boxExtents=<20 20 20>
      constraintSetup {angularLimitsMax=<15 15 15>, angularLimitsMin=<-15 -15 -10>, angularTarget=<1 0 0>}
      localJointOffset=<4 0 0>
     }]
    simSpaceSettings {maxAngularAcceleration=35, maxAngularVelocity=7, simSpaceAngularAlpha=1}
   }
   Save cached pose 'preRBAN'(Pose=Component To Local(ComponentPose=@n21)) {cacheName=preRBAN}
  ]}
 EventGraph {nodes [
   n1 = Event OnBeginSkydiving()
   Set skydiveGravity(execute=[@n1.then, @n2.then, @n3.then_0], skydiveGravity:Vector=SelectVector(A="0,0,300", B="0,0,-980", bPickA=Get bIsSkyDiving())) -> (Output_Get:Vector)
   n2 = Event OnEndSkydiving()
   Event OnCharacterCustomizationLoaded()
   > n3 = Sequence!(execute) -> (then_0, then_1, then_2:Exec, then_3:Exec, then_4:Exec)
   n4 = Get bAnimDynamicsEnabled() -> (bAnimDynamicsEnabled:bool)
   n5 = Get bRigidBodyEnabled()
   n6 = Set rotateHeadbandsLowSettings(execute=@n3.then_2, rotateHeadbandsLowSettings:bool=OR Boolean(A=NOT Boolean(A=@n4), B=NOT Boolean(A=@n5))) -> (Output_Get:bool)
   > Set feLowSettings(feLowSettings:bool=AND Boolean(A=@n6, B=Get bIsInFrontEnd())) -> (Output_Get:bool)
   Set Is Wearing Backbling(execute=@n3.then_3, Is Wearing Backbling=CheckForBackbling())
   n7 = Branch(execute=@n3.then_1, Condition=CharacterItemDefinitionHasAnyMetaTag(Tags=Make Array("[0]"=(TagName="Cosmetics.Variant.Property.Stage1"))))
   > Set faceAcc active?(faceAcc active?:bool=true) -> (Output_Get:bool)
   Set faceAcc active?(execute=@n7.else, faceAcc active?=false)
   n8 = Cast To PlayerPawn_Athena(execute=@n3.then_4, Object=TryGetPawnOwner())
   > Set Body_Mesh(Body_Mesh=GetSkeletalMeshForPartType(self=@n8))
  ]}
]
variables [
 Body_Mesh {type=SkeletalMeshComponent}
 faceAcc active? {type=boolean}
 feLowSettings {type=boolean}
 Is Wearing Backbling {type=boolean}
 rotateHeadbandsLowSettings {type=boolean}
 skydiveGravity {type=Vector, default=<0 0 -980>}
]
